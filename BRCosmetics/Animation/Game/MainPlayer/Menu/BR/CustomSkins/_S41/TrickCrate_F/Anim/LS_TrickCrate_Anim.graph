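# Level Sequence that animates a FortPlayerMannequin for the TrickCrate_F custom skin in the Battle Royale menu, under the _S41 cosmetics folder.
# The mannequin has spawn, transform and attach tracks, plus an MR_Biped_v02 body Control Rig track with two sections.
# A Fortnite_3L_Face_CtrlRig track keys blinks, eye darts, brows, asymmetric mouth corners and a late jaw movement.
# The sequence runs at 30 fps and has no camera or audio tracks.
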LevelSequence /BRCosmetics/Animation/Game/MainPlayer/Menu/BR/CustomSkins/_S41/TrickCrate_F/Anim/LS_TrickCrate_Anim
bindings [FortPlayerMannequin {
  id=#1
  possessed_class=FortPlayerMannequin
  tracks [
   {type=MovieSceneSpawnTrack, sections [{type=MovieSceneSpawnSection, channels [None {type=MovieSceneScriptingBoolChannel, default=true}]}]}
   {
    type=MovieScene3DTransformTrack
    sections [{
      type=MovieScene3DTransformSection
      channels [
       Location.X {type=MovieSceneScriptingDoubleChannel, default=0}
       Location.Y {type=MovieSceneScriptingDoubleChannel, default=0}
       Location.Z {type=MovieSceneScriptingDoubleChannel, default=0}
       Rotation.X {type=MovieSceneScriptingDoubleChannel, default=0}
       Rotation.Y {type=MovieSceneScriptingDoubleChannel, default=0}
       Rotation.Z {type=MovieSceneScriptingDoubleChannel, default=0}
       Scale.X {type=MovieSceneScriptingDoubleChannel, default=1}
       Scale.Y {type=MovieSceneScriptingDoubleChannel, default=1}
       Scale.Z {type=MovieSceneScriptingDoubleChannel, default=1}
      ]
     }]
   }
   {
    type=MovieSceneControlRigParameterTrack
    display_name=MR_Biped_v02
    sections [{type=MovieSceneControlRigParameterSection}, {type=MovieSceneControlRigParameterSection}]
   }
   {
    type=MovieSceneControlRigParameterTrack
    display_name=Fortnite_3L_Face_CtrlRig
    sections [{
      type=MovieSceneControlRigParameterSection
      channels [
       CTRL_C_eyesAim.Location.X {type=MovieSceneScriptingFloatChannel, default=0, const [0,41]}
       CTRL_C_eyesAim.Location.Y {type=MovieSceneScriptingFloatChannel, default=0, const [0,41]}
       CTRL_C_eyesAim.Location.Z {type=MovieSceneScriptingFloatChannel, default=0, const [0,41]}
       CTRL_C_eyesAim.Rotation.X {type=MovieSceneScriptingFloatChannel, default=0, const [0,41]}
       CTRL_C_eyesAim.Rotation.Y {type=MovieSceneScriptingFloatChannel, default=0, const [0,41]}
       CTRL_C_eyesAim.Rotation.Z {type=MovieSceneScriptingFloatChannel, default=0, const [0,41]}
       CTRL_C_eyesAim.Scale.X {type=MovieSceneScriptingFloatChannel, default=1, const [1,41]}
       CTRL_C_eyesAim.Scale.Y {type=MovieSceneScriptingFloatChannel, default=1, const [1,41]}
       CTRL_C_eyesAim.Scale.Z {type=MovieSceneScriptingFloatChannel, default=1, const [1,41]}
       CTRL_L_eyeAim.Location.X {type=MovieSceneScriptingFloatChannel, default=0}
       CTRL_L_eyeAim.Location.Y {type=MovieSceneScriptingFloatChannel, default=0}
       CTRL_L_eyeAim.Location.Z {type=MovieSceneScriptingFloatChannel, default=0}
       CTRL_L_eyeAim.Rotation.X {type=MovieSceneScriptingFloatChannel, default=0}
       CTRL_L_eyeAim.Rotation.Y {type=MovieSceneScriptingFloatChannel, default=0}
       CTRL_L_eyeAim.Rotation.Z {type=MovieSceneScriptingFloatChannel, default=0}
       CTRL_L_eyeAim.Scale.X {type=MovieSceneScriptingFloatChannel, default=1}
       CTRL_L_eyeAim.Scale.Y {type=MovieSceneScriptingFloatChannel, default=1}
       CTRL_L_eyeAim.Scale.Z {type=MovieSceneScriptingFloatChannel, default=1}
       CTRL_R_eyeAim.Location.X {type=MovieSceneScriptingFloatChannel, default=0}
       CTRL_R_eyeAim.Location.Y {type=MovieSceneScriptingFloatChannel, default=0}
       CTRL_R_eyeAim.Location.Z {type=MovieSceneScriptingFloatChannel, default=0}
       CTRL_R_eyeAim.Rotation.X {type=MovieSceneScriptingFloatChannel, default=0}
       CTRL_R_eyeAim.Rotation.Y {type=MovieSceneScriptingFloatChannel, default=0}
       CTRL_R_eyeAim.Rotation.Z {type=MovieSceneScriptingFloatChannel, default=0}
       CTRL_R_eyeAim.Scale.X {type=MovieSceneScriptingFloatChannel, default=1}
       CTRL_R_eyeAim.Scale.Y {type=MovieSceneScriptingFloatChannel, default=1}
       CTRL_R_eyeAim.Scale.Z {type=MovieSceneScriptingFloatChannel, default=1}
       CTRL_convergenceSwitch {type=MovieSceneScriptingFloatChannel, default=0}
       CTRL_Master.Location.X {type=MovieSceneScriptingFloatChannel, default=0}
       CTRL_Master.Location.Y {type=MovieSceneScriptingFloatChannel, default=0}
       CTRL_Master.Location.Z {type=MovieSceneScriptingFloatChannel, default=0}
       CTRL_Master.Rotation.X {type=MovieSceneScriptingFloatChannel, default=0}
       CTRL_Master.Rotation.Y {type=MovieSceneScriptingFloatChannel, default=0}
       CTRL_Master.Rotation.Z {type=MovieSceneScriptingFloatChannel, default=0}
       CTRL_Master.Scale.X {type=MovieSceneScriptingFloatChannel, default=1}
       CTRL_Master.Scale.Y {type=MovieSceneScriptingFloatChannel, default=1}
       CTRL_Master.Scale.Z {type=MovieSceneScriptingFloatChannel, default=1}
       CTRL_C_jaw.X {type=MovieSceneScriptingFloatChannel, default=0, thinned_from=396, keys [[0,0],[395,0]]}
       CTRL_C_jaw.Y {
        type=MovieSceneScriptingFloatChannel
        default=0
        thinned_from=396
        keys [[0,0],[325,0],[326,0.0057],[327,0.0214],[328,0.0447],[329,0.0735],[330,0.1055],[331,0.1383],[332,0.1698],[333,0.1976],[334,0.2196],[335,0.2335],[336,0.2424],[337,0.2509],[338,0.2583],[339,0.2642],[340,0.2682],[341,0.2696],[342,0.2676],[343,0.2621],[344,0.254],[345,0.2442],[346,0.2335],[347,0.2191],[348,0.1986],[349,0.1737],[350,0.1458],[352,0.0874],[353,0.0601],[354,0.0362],[355,0.0171],[356,0.0045],[357,0],[395,0]]
       }
       CTRL_C_jaw_openExtreme {type=MovieSceneScriptingFloatChannel, default=0}
       CTRL_C_jaw_fwdBack {type=MovieSceneScriptingFloatChannel, default=0}
       CTRL_R_jaw_ChinRaiseD {type=MovieSceneScriptingFloatChannel, default=0}
       CTRL_R_jaw_ChinRaiseU {type=MovieSceneScriptingFloatChannel, default=0}
       CTRL_L_jaw_ChinRaiseU {type=MovieSceneScriptingFloatChannel, default=0}
       CTRL_L_jaw_ChinRaiseD {type=MovieSceneScriptingFloatChannel, default=0}
       CTRL_R_mouth_lipsBlow {type=MovieSceneScriptingFloatChannel, default=0}
       CTRL_L_mouth_lipsTogetherU {type=MovieSceneScriptingFloatChannel, default=0, const [0,2]}
       CTRL_L_mouth_lipsTogetherD {type=MovieSceneScriptingFloatChannel, default=0}
       CTRL_R_mouth_lipsTogetherU {type=MovieSceneScriptingFloatChannel, default=0}
       CTRL_L_mouth_lipsBlow {type=MovieSceneScriptingFloatChannel, default=0}
       CTRL_R_mouth_lipsTogetherD {type=MovieSceneScriptingFloatChannel, default=0}
       CTRL_C_tongue_roll.X {type=MovieSceneScriptingFloatChannel, default=0}
       CTRL_C_tongue_roll.Y {type=MovieSceneScriptingFloatChannel, default=0}
       CTRL_C_tongue.X {type=MovieSceneScriptingFloatChannel, default=0}
       CTRL_C_tongue.Y {type=MovieSceneScriptingFloatChannel, default=0}
       CTRL_C_tongue_narrowWide {type=MovieSceneScriptingFloatChannel, default=0}
       CTRL_C_tongue_inOut {type=MovieSceneScriptingFloatChannel, default=0}
       CTRL_R_nose.X {type=MovieSceneScriptingFloatChannel, default=0, const [0,41]}
       CTRL_R_nose.Y {type=MovieSceneScriptingFloatChannel, default=0, const [0,41]}
       CTRL_L_nose.X {type=MovieSceneScriptingFloatChannel, default=0, const [0,41]}
       CTRL_L_nose.Y {type=MovieSceneScriptingFloatChannel, default=0, const [0,41]}
       CTRL_L_eye.X {type=MovieSceneScriptingFloatChannel, default=0, const [0,41]}
       CTRL_L_eye.Y {type=MovieSceneScriptingFloatChannel, default=0, const [0,41]}
       CTRL_L_eye_blink {
        type=MovieSceneScriptingFloatChannel
        default=0.0905
        thinned_from=46
        keys [[0,0.0905],[19,0.0905],[25,1],[29,1],[37,0.0581],[85,0.0581],[90,0.9999],[96,0.1494],[101,0.1494],[114,0.2486],[117,0.2367],[147,0.1945],[165,0.1576],[218,0.1576],[219,0.3948],[220,0.7299],[221,0.914],[222,0.9692],[223,1],[224,0.9512],[225,0.8307],[226,0.4706],[227,0.2947],[228,0.2044],[230,0.1638],[234,0.1512],[315,0.1512],[317,0.9097],[318,0.9115],[320,0.9115],[321,0.6377],[324,0.307],[325,0.2667],[326,0.2419],[330,0.2897],[338,0.1321],[351,0.0905],[395,0.0905]]
       }
       CTRL_R_eye.X {type=MovieSceneScriptingFloatChannel, default=0, const [0,41]}
       CTRL_R_eye.Y {type=MovieSceneScriptingFloatChannel, default=0, const [0,41]}
       CTRL_R_eye_blink {
        type=MovieSceneScriptingFloatChannel
        default=0.0905
        thinned_from=46
        keys [[0,0.0905],[19,0.0905],[25,1],[29,1],[37,0.0581],[85,0.0581],[90,0.9999],[96,0.1494],[101,0.1494],[114,0.2486],[117,0.2367],[147,0.1945],[165,0.1576],[218,0.1576],[219,0.3948],[220,0.7299],[221,0.914],[222,0.9692],[223,1],[224,0.9512],[225,0.8307],[226,0.4706],[227,0.2947],[228,0.2044],[230,0.1638],[234,0.1512],[315,0.1512],[317,0.9097],[318,0.9115],[320,0.9115],[321,0.6377],[324,0.307],[325,0.2667],[326,0.2419],[330,0.2897],[338,0.1321],[351,0.0905],[395,0.0905]]
       }
       CTRL_C_eye.X {
        type=MovieSceneScriptingFloatChannel
        default=-0.4178
        thinned_from=303
        keys [[0,-0.4178],[19,-0.4178],[20,-0.4188],[21,-0.4212],[24,-0.4314],[25,-0.4338],[26,-0.4348],[29,-0.4348],[30,-0.4395],[31,-0.4508],[34,-0.4705],[35,-0.4756],[36,-0.4791],[37,-0.4804],[58,-0.4804],[59,-0.532],[75,-0.532],[76,-0.4804],[85,-0.4804],[86,-0.4456],[87,-0.3578],[96,0.4205],[101,0.4178],[108,0.1578],[114,-0.127],[117,-0.1897],[132,-0.2115],[137,-0.1167],[148,-0.0185],[171,-0.2984],[180,-0.2984],[181,-0.3098],[182,-0.2242],[195,-0.2242],[196,-0.284],[216,-0.284],[217,-0.1114],[218,-0.1049],[219,-0.0881],[220,-0.0823],[221,-0.0809],[224,-0.0809],[225,-0.0137],[226,0.4231],[227,0.5309],[228,0.5459],[232,0.5452],[233,0.5486],[234,0.5498],[258,0.5498],[259,0.4852],[280,0.4852],[281,0.5162],[296,0.5162],[297,0.5498],[318,0.5498],[319,0.3897],[320,0.3897],[321,-0.3304],[324,-0.426],[333,-0.426],[334,-0.3854],[345,-0.3854],[346,-0.412],[348,-0.4157],[350,-0.4176],[395,-0.4178]]
       }
       CTRL_C_eye.Y {
        type=MovieSceneScriptingFloatChannel
        default=-0.2805
        thinned_from=311
        keys [[0,-0.3704],[19,-0.3704],[20,-0.365],[21,-0.3512],[22,-0.3324],[23,-0.3118],[24,-0.2929],[25,-0.2791],[26,-0.2738],[29,-0.2738],[30,-0.2898],[31,-0.3281],[33,-0.3739],[34,-0.3949],[35,-0.4122],[36,-0.4239],[37,-0.4283],[75,-0.4283],[76,-0.388],[85,-0.388],[86,-0.3899],[87,-0.4348],[89,-0.4459],[90,-0.4491],[94,-0.4547],[96,-0.4557],[101,-0.657],[108,-0.4914],[114,-0.6438],[117,-0.6405],[132,-0.5553],[137,-0.5153],[148,-0.4395],[171,-0.3656],[217,-0.3656],[218,-0.3641],[219,-0.3602],[220,-0.3588],[224,-0.3585],[225,-0.3429],[226,-0.272],[227,-0.2439],[228,-0.2392],[232,-0.239],[233,-0.2382],[318,-0.2379],[319,-0.2579],[320,-0.2579],[321,-0.3618],[324,-0.408],[325,-0.42],[326,-0.4297],[331,-0.4293],[333,-0.4264],[335,-0.4212],[345,-0.3822],[347,-0.376],[349,-0.3719],[351,-0.3704],[395,-0.3704]]
       }
       CTRL_C_eyesConvergence {type=MovieSceneScriptingFloatChannel, default=0, const [0,41]}
       CTRL_L_eye_cheekRaise {type=MovieSceneScriptingFloatChannel, default=0}
       CTRL_L_eye_squintInner {type=MovieSceneScriptingFloatChannel, default=0, const [0,41]}
       CTRL_L_nose_wrinkleUpper {type=MovieSceneScriptingFloatChannel, default=0, const [0,41]}
       CTRL_R_nose_wrinkleUpper {type=MovieSceneScriptingFloatChannel, default=0, const [0,41]}
       CTRL_R_eye_squintInner {type=MovieSceneScriptingFloatChannel, default=0, const [0,41]}
       CTRL_R_eye_cheekRaise {type=MovieSceneScriptingFloatChannel, default=0}
       CTRL_C_eyelidsReduce {type=MovieSceneScriptingFloatChannel, default=0}
       CTRL_R_brow_lateral {type=MovieSceneScriptingFloatChannel, default=0, const [0,7]}
       CTRL_L_brow_lateral {type=MovieSceneScriptingFloatChannel, default=0, const [0,7]}
       CTRL_R_brow_down {
        type=MovieSceneScriptingFloatChannel
        default=0
        thinned_from=283
        keys [[0,0.252],[20,0.252],[21,0.2715],[22,0.3349],[23,0.4244],[24,0.5228],[25,0.6124],[26,0.6757],[27,0.6953],[29,0.6789],[30,0.5876],[31,0.381],[33,0.1501],[87,0.1501],[90,0.7699],[91,0.7869],[92,0.7311],[94,0.3479],[220,0.3479],[221,0.4077],[223,0.482],[225,0.4644],[226,0.4361],[227,0.2908],[228,0.2563],[370,0.2563],[371,0.2541],[394,0.2541],[395,0.252]]
       }
       CTRL_L_brow_down {
        type=MovieSceneScriptingFloatChannel
        default=0
        thinned_from=282
        keys [[0,0.3828],[20,0.3828],[22,0.5786],[23,0.6682],[24,0.7665],[25,0.8561],[26,0.9194],[27,0.939],[29,0.9227],[30,0.8313],[31,0.6247],[34,0.3767],[88,0.3767],[90,0.7699],[91,0.7869],[92,0.7311],[94,0.499],[220,0.499],[221,0.5792],[223,0.6176],[225,0.5926],[226,0.5527],[227,0.3828],[395,0.3828]]
       }
       CTRL_L_brow_raiseOut {type=MovieSceneScriptingFloatChannel, default=0}
       CTRL_L_brow_raiseIn {type=MovieSceneScriptingFloatChannel, default=0}
       CTRL_R_brow_raiseIn {type=MovieSceneScriptingFloatChannel, default=0}
       CTRL_R_brow_raiseOut {type=MovieSceneScriptingFloatChannel, default=0}
       CTRL_lookAtSwitch {type=MovieSceneScriptingFloatChannel, default=0, const [0,2]}
       CTRL_C_teethD.X {type=MovieSceneScriptingFloatChannel, default=0}
       CTRL_C_teethD.Y {type=MovieSceneScriptingFloatChannel, default=0}
       CTRL_C_teethU.X {type=MovieSceneScriptingFloatChannel, default=0}
       CTRL_C_teethU.Y {type=MovieSceneScriptingFloatChannel, default=0}
       CTRL_C_teeth_fwdBackU {type=MovieSceneScriptingFloatChannel, default=0}
       CTRL_L_mouth_corner.X {
        type=MovieSceneScriptingFloatChannel
        default=0.3013
        thinned_from=396
        keys [[0,0.3013],[7,0.3013],[8,0.3038],[9,0.3106],[10,0.3211],[11,0.3344],[12,0.3496],[15,0.3994],[16,0.4147],[17,0.428],[18,0.4385],[19,0.4454],[20,0.4479],[59,0.4492],[60,0.4545],[61,0.4688],[62,0.4899],[63,0.5155],[65,0.5712],[66,0.5968],[67,0.6178],[68,0.6321],[69,0.6374],[325,0.6374],[326,0.6228],[327,0.5824],[328,0.5214],[329,0.4448],[330,0.3579],[332,0.1733],[333,0.0858],[334,0.0085],[335,-0.0536],[336,-0.0954],[337,-0.1118],[340,-0.1173],[342,-0.1203],[343,-0.1215],[344,-0.1224],[345,-0.123],[346,-0.1232],[347,-0.1133],[348,-0.0862],[349,-0.0457],[350,0.0044],[351,0.0602],[352,0.1179],[353,0.1737],[354,0.2238],[355,0.2643],[356,0.2914],[357,0.3013],[395,0.3013]]
       }
       CTRL_L_mouth_corner.Y {
        type=MovieSceneScriptingFloatChannel
        default=0.8278
        thinned_from=396
        keys [[0,0.8278],[7,0.8278],[8,0.8214],[9,0.8036],[10,0.7766],[11,0.7423],[12,0.7028],[15,0.5741],[16,0.5346],[17,0.5002],[18,0.4731],[19,0.4552],[20,0.4487],[59,0.4454],[60,0.4509],[61,0.4657],[62,0.4877],[63,0.5143],[65,0.5722],[66,0.5988],[67,0.6208],[68,0.6356],[69,0.6411],[346,0.6417],[347,0.646],[348,0.6579],[349,0.6757],[350,0.6978],[353,0.7722],[354,0.7942],[355,0.812],[356,0.824],[357,0.8283],[395,0.8283]]
       }
       CTRL_R_mouth_corner.X {
        type=MovieSceneScriptingFloatChannel
        default=0.1741
        thinned_from=396
        keys [[0,0.1741],[7,0.1741],[8,0.1834],[9,0.2094],[10,0.249],[11,0.2992],[12,0.3571],[13,0.4194],[14,0.4833],[15,0.5457],[16,0.6036],[17,0.6539],[18,0.6936],[19,0.7198],[20,0.7293],[59,0.7342],[60,0.7185],[61,0.676],[62,0.6132],[63,0.537],[65,0.3712],[66,0.295],[67,0.2323],[68,0.1897],[69,0.1741],[325,0.1741],[326,0.1639],[327,0.1355],[328,0.0923],[329,0.0374],[330,-0.0259],[331,-0.0942],[332,-0.1643],[333,-0.233],[334,-0.2969],[335,-0.3528],[336,-0.3974],[337,-0.4274],[339,-0.47],[340,-0.4899],[341,-0.5083],[342,-0.5245],[343,-0.538],[344,-0.5484],[345,-0.555],[346,-0.5573],[347,-0.5402],[348,-0.4935],[349,-0.4238],[350,-0.3375],[351,-0.2413],[352,-0.1419],[353,-0.0457],[354,0.0405],[355,0.1103],[356,0.157],[357,0.1741],[395,0.1741]]
       }
       CTRL_R_mouth_corner.Y {type=MovieSceneScriptingFloatChannel, default=0.6585, thinned_from=396, keys [[0,0.6585],[395,0.6591]]}
       CTRL_C_teeth_fwdBackD {type=MovieSceneScriptingFloatChannel, default=0}
       CTRL_R_mouth_cornerSharpnessD {type=MovieSceneScriptingFloatChannel, default=0}
       CTRL_R_mouth_cornerSharpnessU {type=MovieSceneScriptingFloatChannel, default=0}
       CTRL_L_eye_eyelidU.X {
        type=MovieSceneScriptingFloatChannel
        default=0.9047
        thinned_from=41
        keys [[0,0.9047],[19,0.9047],[25,0.738],[29,0.738],[37,0.2895],[85,0.2895],[90,0],[96,-0.1316],[101,0.1388],[320,0.1388],[321,0.2567],[324,0.5659],[325,0.7064],[326,0.7293],[330,0.7293],[351,0],[372,0.9047],[395,0.9047]]
       }
       CTRL_L_eye_eyelidU.Y {
        type=MovieSceneScriptingFloatChannel
        default=0
        thinned_from=41
        keys [[0,0],[90,0],[96,0.0209],[101,0.0825],[114,0.0825],[147,-0.014],[221,-0.014],[223,-0.1135],[224,-0.1184],[225,-0.0986],[226,-0.0319],[228,0.1203],[234,0.1365],[315,0.1365],[317,0.0003],[318,0],[320,0],[321,-0.0161],[324,-0.0354],[325,-0.0442],[326,-0.0457],[330,-0.0457],[351,0.1365],[372,0],[395,0]]
       }
       CTRL_L_mouth_cornerSharpnessD {type=MovieSceneScriptingFloatChannel, default=0}
       CTRL_L_eye_eyelidD.X {type=MovieSceneScriptingFloatChannel, default=0, thinned_from=41, keys [[0,0],[395,0]]}
       CTRL_L_eye_eyelidD.Y {
        type=MovieSceneScriptingFloatChannel
        default=-0.2394
        thinned_from=41
        keys [[0,-0.2394],[19,-0.2394],[25,-0.1745],[29,-0.1745],[37,-0.288],[85,-0.288],[90,0],[96,-0.273],[101,-0.2414],[315,-0.2414],[317,-0.0005],[318,0],[320,0],[321,-0.0847],[324,-0.1868],[325,-0.2332],[326,-0.2407],[351,-0.2394],[395,-0.2394]]
       }
       CTRL_L_mouth_cornerSharpnessU {type=MovieSceneScriptingFloatChannel, default=0}
       CTRL_R_mouth_lipsTowardsTeethD {type=MovieSceneScriptingFloatChannel, default=0}
       CTRL_L_mouth_lipsTowardsTeethD {type=MovieSceneScriptingFloatChannel, default=0}
       CTRL_R_eye_eyelidU.X {type=MovieSceneScriptingFloatChannel, default=0, thinned_from=41, keys [[0,0],[395,0]]}
       CTRL_R_eye_eyelidU.Y {
        type=MovieSceneScriptingFloatChannel
        default=0
        thinned_from=41
        keys [[0,0],[90,0],[96,0.0209],[101,0.0825],[114,0.0825],[147,-0.0064],[221,-0.0064],[223,-0.1135],[224,-0.1184],[225,-0.0986],[226,-0.0319],[228,0.1203],[234,0.1365],[315,0.1365],[317,0.0003],[318,0],[320,0],[321,0.0311],[324,0.0687],[325,0.0857],[326,0.0885],[330,0.0885],[351,0],[395,0]]
       }
       CTRL_L_mouth_lipsTowardsTeethU {type=MovieSceneScriptingFloatChannel, default=0}
       CTRL_R_eye_eyelidD.X {
        type=MovieSceneScriptingFloatChannel
        default=0
        thinned_from=41
        keys [[0,0],[19,0],[25,0.1671],[29,0.1671],[37,0.0852],[85,0.0852],[90,0],[96,0.0578],[101,0],[395,0]]
       }
       CTRL_R_eye_eyelidD.Y {
        type=MovieSceneScriptingFloatChannel
        default=-0.2394
        thinned_from=41
        keys [[0,-0.2394],[19,-0.2394],[25,-0.1745],[29,-0.1745],[37,-0.288],[85,-0.288],[90,0],[96,-0.273],[101,-0.2414],[315,-0.2414],[317,-0.0005],[318,0],[320,0],[321,-0.0847],[324,-0.1868],[325,-0.2332],[326,-0.2407],[351,-0.2394],[395,-0.2394]]
       }
       CTRL_R_mouth_lipsTowardsTeethU {type=MovieSceneScriptingFloatChannel, default=0}
       CTRL_faceGUIfollowHead {type=MovieSceneScriptingFloatChannel, default=1}
       CTRL_eyesAimFollowHead {type=MovieSceneScriptingFloatChannel, default=0, keys [[90,0],[91,0],[189,1]]}
       CTRL_C_nose.X {type=MovieSceneScriptingFloatChannel, default=0}
       CTRL_C_nose.Y {type=MovieSceneScriptingFloatChannel, default=0}
       CTRL_C_teeth_rotateU {type=MovieSceneScriptingFloatChannel, default=0}
       CTRL_C_teeth_rotateD {type=MovieSceneScriptingFloatChannel, default=0}
       CTRL_L_brow_inner {type=MovieSceneScriptingFloatChannel, default=0, const [0,7]}
       CTRL_R_brow_inner {type=MovieSceneScriptingFloatChannel, default=0, const [0,7]}
       CTRL_L_brow_outer {type=MovieSceneScriptingFloatChannel, default=0, const [0,7]}
       CTRL_R_brow_outer {type=MovieSceneScriptingFloatChannel, default=0, const [0,7]}
       CTRL_C_mouth.X {type=MovieSceneScriptingFloatChannel, default=0, const [0,396]}
       CTRL_C_mouth.Y {type=MovieSceneScriptingFloatChannel, default=0, const [0,396]}
       CTRL_L_mouth_upperLipRaise {type=MovieSceneScriptingFloatChannel, default=0, const [0,396]}
       CTRL_L_mouth_cornerPull {type=MovieSceneScriptingFloatChannel, default=0, const [0,396]}
       CTRL_L_mouth_dimple {
        type=MovieSceneScriptingFloatChannel
        default=0
        thinned_from=396
        keys [[0,0],[7,0],[8,0.0019],[9,0.0071],[10,0.0151],[11,0.0252],[12,0.0368],[13,0.0494],[14,0.0622],[15,0.0748],[16,0.0864],[17,0.0966],[18,0.1045],[19,0.1098],[20,0.1117],[59,0.1127],[60,0.1096],[61,0.101],[62,0.0884],[63,0.073],[65,0.0397],[66,0.0243],[67,0.0117],[68,0.0032],[69,0],[395,0]]
       }
       CTRL_L_mouth_cornerDepress {type=MovieSceneScriptingFloatChannel, default=0, const [0,396]}
       CTRL_L_mouth_stretch {type=MovieSceneScriptingFloatChannel, default=0, const [0,396]}
       CTRL_L_mouth_lowerLipDepress {type=MovieSceneScriptingFloatChannel, default=0, const [0,396]}
       CTRL_R_mouth_lowerLipDepress {type=MovieSceneScriptingFloatChannel, default=0, const [0,396]}
       CTRL_R_mouth_stretch {type=MovieSceneScriptingFloatChannel, default=0, const [0,396]}
       CTRL_L_mouth_towardsD {type=MovieSceneScriptingFloatChannel, default=0}
       CTRL_R_mouth_funnelU {type=MovieSceneScriptingFloatChannel, default=0}
       CTRL_L_mouth_towardsU {type=MovieSceneScriptingFloatChannel, default=0, const [0,2]}
       CTRL_R_mouth_lipBiteD {type=MovieSceneScriptingFloatChannel, default=0}
       CTRL_R_mouth_funnelD {type=MovieSceneScriptingFloatChannel, default=0}
       CTRL_L_mouth_pressU {type=MovieSceneScriptingFloatChannel, default=0}
       CTRL_L_mouth_funnelD {type=MovieSceneScriptingFloatChannel, default=0}
       CTRL_L_mouth_suckBlow {type=MovieSceneScriptingFloatChannel, default=0}
       CTRL_L_mouth_pressD {type=MovieSceneScriptingFloatChannel, default=0}
       CTRL_L_mouth_funnelU {type=MovieSceneScriptingFloatChannel, default=0}
       CTRL_R_mouth_towardsU {type=MovieSceneScriptingFloatChannel, default=0}
       CTRL_R_mouth_purseD {type=MovieSceneScriptingFloatChannel, default=0}
       CTRL_L_mouth_lipBiteU {type=MovieSceneScriptingFloatChannel, default=0}
       CTRL_R_mouth_pressD {type=MovieSceneScriptingFloatChannel, default=0}
       CTRL_R_mouth_pressU {type=MovieSceneScriptingFloatChannel, default=0}
       CTRL_L_mouth_purseD {type=MovieSceneScriptingFloatChannel, default=0}
       CTRL_R_mouth_lipBiteU {type=MovieSceneScriptingFloatChannel, default=0}
       CTRL_R_mouth_suckBlow {type=MovieSceneScriptingFloatChannel, default=0}
       CTRL_L_mouth_purseU {type=MovieSceneScriptingFloatChannel, default=0}
       CTRL_R_mouth_towardsD {type=MovieSceneScriptingFloatChannel, default=0}
       CTRL_L_mouth_lipBiteD {type=MovieSceneScriptingFloatChannel, default=0}
       CTRL_R_mouth_purseU {type=MovieSceneScriptingFloatChannel, default=0}
       CTRL_R_mouth_cornerDepress {type=MovieSceneScriptingFloatChannel, default=0, const [0,396]}
       CTRL_R_mouth_upperLipRaise {type=MovieSceneScriptingFloatChannel, default=0, const [0,396]}
       CTRL_R_mouth_dimple {
        type=MovieSceneScriptingFloatChannel
        default=0
        thinned_from=396
        keys [[0,0],[7,0],[8,0.0048],[9,0.0182],[10,0.0385],[11,0.0644],[12,0.0941],[13,0.1261],[14,0.1589],[15,0.191],[16,0.2207],[17,0.2465],[18,0.2669],[19,0.2803],[20,0.2851],[59,0.2851],[60,0.2771],[61,0.2554],[62,0.2235],[63,0.1847],[65,0.1003],[66,0.0616],[67,0.0296],[68,0.008],[69,0],[395,0]]
       }
       CTRL_R_mouth_cornerPull {type=MovieSceneScriptingFloatChannel, default=0, const [0,396]}
       Weight {type=MovieSceneScriptingFloatChannel, default=1}
      ]
     }]
   }
   {type=MovieScene3DAttachTrack, sections [{type=MovieScene3DAttachSection}]}
  ]
 }]
display_rate {denominator=1, numerator=30}
tick_resolution {denominator=1, numerator=24000}
tracks []
