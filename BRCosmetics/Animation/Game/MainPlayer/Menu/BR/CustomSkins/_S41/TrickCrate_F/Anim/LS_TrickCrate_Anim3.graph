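# Level Sequence for authoring the third menu animation of the TrickCrate_F custom skin on a spawnable FortPlayerMannequin at the origin, at 30 fps.
# Control Rig tracks drive the body through MR_Biped_v02 (two sections, no channels shown) and the face through Fortnite_3L_Face_CtrlRig.
# Facial motion comes from keyed gaze, blinks, eyelids and brows over a constant, asymmetric mouth-corner pose.
# Many of those keyed channels return to their opening values by the last key.
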LevelSequence /BRCosmetics/Animation/Game/MainPlayer/Menu/BR/CustomSkins/_S41/TrickCrate_F/Anim/LS_TrickCrate_Anim3
bindings [FortPlayerMannequin {
  id=#1
  possessed_class=FortPlayerMannequin
  tracks [
   {type=MovieSceneSpawnTrack, sections [{type=MovieSceneSpawnSection, channels [None {type=MovieSceneScriptingBoolChannel, default=true}]}]}
   {
    type=MovieScene3DTransformTrack
    sections [{
      type=MovieScene3DTransformSection
      channels [
       Location.X {type=MovieSceneScriptingDoubleChannel, default=0}
       Location.Y {type=MovieSceneScriptingDoubleChannel, default=0}
       Location.Z {type=MovieSceneScriptingDoubleChannel, default=0}
       Rotation.X {type=MovieSceneScriptingDoubleChannel, default=0}
       Rotation.Y {type=MovieSceneScriptingDoubleChannel, default=0}
       Rotation.Z {type=MovieSceneScriptingDoubleChannel, default=0}
       Scale.X {type=MovieSceneScriptingDoubleChannel, default=1}
       Scale.Y {type=MovieSceneScriptingDoubleChannel, default=1}
       Scale.Z {type=MovieSceneScriptingDoubleChannel, default=1}
      ]
     }]
   }
   {
    type=MovieSceneControlRigParameterTrack
    display_name=MR_Biped_v02
    sections [{type=MovieSceneControlRigParameterSection}, {type=MovieSceneControlRigParameterSection}]
   }
   {
    type=MovieSceneControlRigParameterTrack
    display_name=Fortnite_3L_Face_CtrlRig
    sections [{
      type=MovieSceneControlRigParameterSection
      channels [
       CTRL_C_eyesAim.Location.X {type=MovieSceneScriptingFloatChannel, default=0, const [0,15]}
       CTRL_C_eyesAim.Location.Y {type=MovieSceneScriptingFloatChannel, default=0, const [0,15]}
       CTRL_C_eyesAim.Location.Z {type=MovieSceneScriptingFloatChannel, default=0, const [0,15]}
       CTRL_C_eyesAim.Rotation.X {type=MovieSceneScriptingFloatChannel, default=0, const [0,15]}
       CTRL_C_eyesAim.Rotation.Y {type=MovieSceneScriptingFloatChannel, default=0, const [0,15]}
       CTRL_C_eyesAim.Rotation.Z {type=MovieSceneScriptingFloatChannel, default=0, const [0,15]}
       CTRL_C_eyesAim.Scale.X {type=MovieSceneScriptingFloatChannel, default=1, const [1,15]}
       CTRL_C_eyesAim.Scale.Y {type=MovieSceneScriptingFloatChannel, default=1, const [1,15]}
       CTRL_C_eyesAim.Scale.Z {type=MovieSceneScriptingFloatChannel, default=1, const [1,15]}
       CTRL_L_eyeAim.Location.X {type=MovieSceneScriptingFloatChannel, default=0, const [0,12]}
       CTRL_L_eyeAim.Location.Y {type=MovieSceneScriptingFloatChannel, default=0, const [0,12]}
       CTRL_L_eyeAim.Location.Z {type=MovieSceneScriptingFloatChannel, default=0, const [0,12]}
       CTRL_L_eyeAim.Rotation.X {type=MovieSceneScriptingFloatChannel, default=0, const [0,12]}
       CTRL_L_eyeAim.Rotation.Y {type=MovieSceneScriptingFloatChannel, default=0, const [0,12]}
       CTRL_L_eyeAim.Rotation.Z {type=MovieSceneScriptingFloatChannel, default=0, const [0,12]}
       CTRL_L_eyeAim.Scale.X {type=MovieSceneScriptingFloatChannel, default=1, const [1,12]}
       CTRL_L_eyeAim.Scale.Y {type=MovieSceneScriptingFloatChannel, default=1, const [1,12]}
       CTRL_L_eyeAim.Scale.Z {type=MovieSceneScriptingFloatChannel, default=1, const [1,12]}
       CTRL_R_eyeAim.Location.X {type=MovieSceneScriptingFloatChannel, default=0, const [0,12]}
       CTRL_R_eyeAim.Location.Y {type=MovieSceneScriptingFloatChannel, default=0, const [0,12]}
       CTRL_R_eyeAim.Location.Z {type=MovieSceneScriptingFloatChannel, default=0, const [0,12]}
       CTRL_R_eyeAim.Rotation.X {type=MovieSceneScriptingFloatChannel, default=0, const [0,12]}
       CTRL_R_eyeAim.Rotation.Y {type=MovieSceneScriptingFloatChannel, default=0, const [0,12]}
       CTRL_R_eyeAim.Rotation.Z {type=MovieSceneScriptingFloatChannel, default=0, thinned_from=12, keys [[0,0],[274,0]]}
       CTRL_R_eyeAim.Scale.X {type=MovieSceneScriptingFloatChannel, default=1, const [1,12]}
       CTRL_R_eyeAim.Scale.Y {type=MovieSceneScriptingFloatChannel, default=1, const [1,12]}
       CTRL_R_eyeAim.Scale.Z {type=MovieSceneScriptingFloatChannel, default=1, const [1,12]}
       CTRL_convergenceSwitch {type=MovieSceneScriptingFloatChannel, default=0, const [0,12]}
       CTRL_Master.Location.X {type=MovieSceneScriptingFloatChannel, default=0, const [0,12]}
       CTRL_Master.Location.Y {type=MovieSceneScriptingFloatChannel, default=0, const [0,12]}
       CTRL_Master.Location.Z {type=MovieSceneScriptingFloatChannel, default=0, const [0,12]}
       CTRL_Master.Rotation.X {type=MovieSceneScriptingFloatChannel, default=0, const [0,12]}
       CTRL_Master.Rotation.Y {type=MovieSceneScriptingFloatChannel, default=0, const [0,12]}
       CTRL_Master.Rotation.Z {type=MovieSceneScriptingFloatChannel, default=0, const [0,12]}
       CTRL_Master.Scale.X {type=MovieSceneScriptingFloatChannel, default=1, const [1,12]}
       CTRL_Master.Scale.Y {type=MovieSceneScriptingFloatChannel, default=1, const [1,12]}
       CTRL_Master.Scale.Z {type=MovieSceneScriptingFloatChannel, default=1, const [1,12]}
       CTRL_C_jaw.X {type=MovieSceneScriptingFloatChannel, default=0, const [0,17]}
       CTRL_C_jaw.Y {type=MovieSceneScriptingFloatChannel, default=0, const [0,17]}
       CTRL_C_jaw_openExtreme {type=MovieSceneScriptingFloatChannel, default=0, const [0,17]}
       CTRL_C_jaw_fwdBack {type=MovieSceneScriptingFloatChannel, default=0, const [0,16]}
       CTRL_R_jaw_ChinRaiseD {type=MovieSceneScriptingFloatChannel, default=0, const [0,17]}
       CTRL_R_jaw_ChinRaiseU {type=MovieSceneScriptingFloatChannel, default=0, const [0,17]}
       CTRL_L_jaw_ChinRaiseU {type=MovieSceneScriptingFloatChannel, default=0, const [0,17]}
       CTRL_L_jaw_ChinRaiseD {type=MovieSceneScriptingFloatChannel, default=0, const [0,17]}
       CTRL_R_mouth_lipsBlow {type=MovieSceneScriptingFloatChannel, default=0, const [0,17]}
       CTRL_L_mouth_lipsTogetherU {type=MovieSceneScriptingFloatChannel, default=0, const [0,17]}
       CTRL_L_mouth_lipsTogetherD {type=MovieSceneScriptingFloatChannel, default=0, const [0,17]}
       CTRL_R_mouth_lipsTogetherU {type=MovieSceneScriptingFloatChannel, default=0, const [0,17]}
       CTRL_L_mouth_lipsBlow {type=MovieSceneScriptingFloatChannel, default=0, const [0,17]}
       CTRL_R_mouth_lipsTogetherD {type=MovieSceneScriptingFloatChannel, default=0, const [0,17]}
       CTRL_C_tongue_roll.X {type=MovieSceneScriptingFloatChannel, default=0, const [0,17]}
       CTRL_C_tongue_roll.Y {type=MovieSceneScriptingFloatChannel, default=0, const [0,17]}
       CTRL_C_tongue.X {type=MovieSceneScriptingFloatChannel, default=0, const [0,17]}
       CTRL_C_tongue.Y {type=MovieSceneScriptingFloatChannel, default=0, const [0,17]}
       CTRL_C_tongue_narrowWide {type=MovieSceneScriptingFloatChannel, default=0, const [0,16]}
       CTRL_C_tongue_inOut {type=MovieSceneScriptingFloatChannel, default=0, const [0,12]}
       CTRL_R_nose.X {type=MovieSceneScriptingFloatChannel, default=0, const [0,20]}
       CTRL_R_nose.Y {type=MovieSceneScriptingFloatChannel, default=0, const [0,20]}
       CTRL_L_nose.X {type=MovieSceneScriptingFloatChannel, default=0, const [0,20]}
       CTRL_L_nose.Y {type=MovieSceneScriptingFloatChannel, default=0, const [0,20]}
       CTRL_L_eye.X {type=MovieSceneScriptingFloatChannel, default=0, const [0,20]}
       CTRL_L_eye.Y {type=MovieSceneScriptingFloatChannel, default=0, const [0,20]}
       CTRL_L_eye_blink {
        type=MovieSceneScriptingFloatChannel
        default=0
        thinned_from=20
        keys [[0,0.0905],[19,0.0905],[26,1],[35,1],[45,0.0581],[85,0.0581],[90,0.9999],[96,0.1269],[101,0.1188],[114,0.2261],[165,0.097],[218,0.097],[239,0.1512],[316,0.1512],[351,0.0905],[379,0.0905]]
       }
       CTRL_R_eye.X {type=MovieSceneScriptingFloatChannel, default=0, const [0,20]}
       CTRL_R_eye.Y {type=MovieSceneScriptingFloatChannel, default=0, const [0,20]}
       CTRL_R_eye_blink {
        type=MovieSceneScriptingFloatChannel
        default=0
        thinned_from=20
        keys [[0,0.0905],[19,0.0905],[26,1],[35,1],[45,0.0581],[85,0.0581],[90,0.9999],[96,0.1269],[101,0.1188],[114,0.2261],[165,0.097],[218,0.097],[239,0.1512],[316,0.1512],[351,0.0905],[379,0.0905]]
       }
       CTRL_C_eye.X {
        type=MovieSceneScriptingFloatChannel
        default=0
        thinned_from=20
        keys [[0,-0.4178],[19,-0.4178],[26,-0.4348],[35,-0.4348],[45,-0.4804],[85,-0.4804],[90,-0.0223],[96,0.1524],[101,0.1712],[114,0],[218,0],[239,0.5498],[316,0.5498],[351,-0.4178],[379,-0.4178]]
       }
       CTRL_C_eye.Y {
        type=MovieSceneScriptingFloatChannel
        default=0
        thinned_from=20
        keys [[0,-0.2805],[19,-0.2805],[26,-0.1839],[35,-0.1839],[45,-0.3384],[85,-0.3384],[90,-0.3592],[96,-0.4079],[101,-0.5539],[114,-0.5539],[165,-0.2757],[218,-0.2757],[239,-0.1481],[316,-0.1481],[351,-0.2805],[379,-0.2805]]
       }
       CTRL_C_eyesConvergence {type=MovieSceneScriptingFloatChannel, default=0, const [0,15]}
       CTRL_L_eye_cheekRaise {type=MovieSceneScriptingFloatChannel, default=0, const [0,17]}
       CTRL_L_eye_squintInner {type=MovieSceneScriptingFloatChannel, default=0, const [0,20]}
       CTRL_L_nose_wrinkleUpper {type=MovieSceneScriptingFloatChannel, default=0, const [0,20]}
       CTRL_R_nose_wrinkleUpper {type=MovieSceneScriptingFloatChannel, default=0, const [0,20]}
       CTRL_R_eye_squintInner {type=MovieSceneScriptingFloatChannel, default=0, const [0,20]}
       CTRL_R_eye_cheekRaise {type=MovieSceneScriptingFloatChannel, default=0, const [0,17]}
       CTRL_C_eyelidsReduce {type=MovieSceneScriptingFloatChannel, default=0, const [0,16]}
       CTRL_R_brow_lateral {type=MovieSceneScriptingFloatChannel, default=0, const [0,20]}
       CTRL_L_brow_lateral {type=MovieSceneScriptingFloatChannel, default=0, const [0,20]}
       CTRL_R_brow_down {
        type=MovieSceneScriptingFloatChannel
        default=0
        thinned_from=20
        keys [[0,0.3049],[19,0.3049],[26,0.468],[35,0.468],[45,0.1036],[316,0.1036],[351,0.3049],[379,0.3049]]
       }
       CTRL_L_brow_down {
        type=MovieSceneScriptingFloatChannel
        default=0
        thinned_from=20
        keys [[0,0.6628],[19,0.6628],[26,0.9055],[35,0.9055],[45,0.7549],[316,0.7549],[351,0.6628],[379,0.6628]]
       }
       CTRL_L_brow_raiseOut {type=MovieSceneScriptingFloatChannel, default=0, thinned_from=18, keys [[0,0],[85,0],[96,0.0904],[114,0],[379,0]]}
       CTRL_L_brow_raiseIn {type=MovieSceneScriptingFloatChannel, default=0, thinned_from=18, keys [[0,0.061],[85,0.061],[96,0.0904],[114,0.061],[379,0.061]]}
       CTRL_R_brow_raiseIn {type=MovieSceneScriptingFloatChannel, default=0, const [0,17]}
       CTRL_R_brow_raiseOut {type=MovieSceneScriptingFloatChannel, default=0, const [0,17]}
       CTRL_lookAtSwitch {type=MovieSceneScriptingFloatChannel, default=0, const [0,13]}
       CTRL_C_teethD.X {type=MovieSceneScriptingFloatChannel, default=0, const [0,17]}
       CTRL_C_teethD.Y {type=MovieSceneScriptingFloatChannel, default=0, const [0,17]}
       CTRL_C_teethU.X {type=MovieSceneScriptingFloatChannel, default=0, const [0,17]}
       CTRL_C_teethU.Y {type=MovieSceneScriptingFloatChannel, default=0, const [0,17]}
       CTRL_C_teeth_fwdBackU {type=MovieSceneScriptingFloatChannel, default=0, const [0,16]}
       CTRL_L_mouth_corner.X {type=MovieSceneScriptingFloatChannel, default=0, const [0.3003,17]}
       CTRL_L_mouth_corner.Y {type=MovieSceneScriptingFloatChannel, default=0, const [0.5488,17]}
       CTRL_R_mouth_corner.X {type=MovieSceneScriptingFloatChannel, default=0, const [0.7626,17]}
       CTRL_R_mouth_corner.Y {type=MovieSceneScriptingFloatChannel, default=0, const [0.5445,17]}
       CTRL_C_teeth_fwdBackD {type=MovieSceneScriptingFloatChannel, default=0, const [0,12]}
       CTRL_R_mouth_cornerSharpnessD {type=MovieSceneScriptingFloatChannel, default=0, const [0,12]}
       CTRL_R_mouth_cornerSharpnessU {type=MovieSceneScriptingFloatChannel, default=0, const [0,12]}
       CTRL_L_eye_eyelidU.X {
        type=MovieSceneScriptingFloatChannel
        default=0
        thinned_from=17
        keys [[0,0.9047],[19,0.9047],[26,0.738],[35,0.738],[45,0.2895],[85,0.2895],[90,0],[96,0.1962],[101,0],[274,0]]
       }
       CTRL_L_eye_eyelidU.Y {
        type=MovieSceneScriptingFloatChannel
        default=0
        thinned_from=17
        keys [[0,0],[90,0],[96,0.0209],[101,0.0825],[114,0.0825],[147,0.1365],[274,0.1365]]
       }
       CTRL_L_mouth_cornerSharpnessD {type=MovieSceneScriptingFloatChannel, default=0, const [0,12]}
       CTRL_L_eye_eyelidD.X {type=MovieSceneScriptingFloatChannel, default=0, thinned_from=20, keys [[0,0],[379,0]]}
       CTRL_L_eye_eyelidD.Y {
        type=MovieSceneScriptingFloatChannel
        default=0
        thinned_from=20
        keys [[0,-0.2394],[19,-0.2394],[26,-0.1745],[35,-0.1745],[45,-0.288],[85,-0.288],[90,0],[96,-0.273],[101,-0.2414],[316,-0.2414],[351,-0.2394],[379,-0.2394]]
       }
       CTRL_L_mouth_cornerSharpnessU {type=MovieSceneScriptingFloatChannel, default=0, const [0,16]}
       CTRL_R_mouth_lipsTowardsTeethD {type=MovieSceneScriptingFloatChannel, default=0, const [0,17]}
       CTRL_L_mouth_lipsTowardsTeethD {type=MovieSceneScriptingFloatChannel, default=0, const [0,17]}
       CTRL_R_eye_eyelidU.X {type=MovieSceneScriptingFloatChannel, default=0, thinned_from=21, keys [[0,0],[379,0]]}
       CTRL_R_eye_eyelidU.Y {
        type=MovieSceneScriptingFloatChannel
        default=0
        thinned_from=21
        keys [[0,0],[90,0],[96,0.0209],[101,0.0825],[114,0.0825],[147,0.1365],[316,0.1365],[351,0],[379,0]]
       }
       CTRL_L_mouth_lipsTowardsTeethU {type=MovieSceneScriptingFloatChannel, default=0, const [0,13]}
       CTRL_R_eye_eyelidD.X {
        type=MovieSceneScriptingFloatChannel
        default=0
        thinned_from=20
        keys [[0,0],[19,0],[26,0.1671],[35,0.1671],[45,0.0852],[85,0.0852],[90,0],[96,0.0578],[101,0],[379,0]]
       }
       CTRL_R_eye_eyelidD.Y {
        type=MovieSceneScriptingFloatChannel
        default=0
        thinned_from=20
        keys [[0,-0.2394],[19,-0.2394],[26,-0.1745],[35,-0.1745],[45,-0.288],[85,-0.288],[90,0],[96,-0.273],[101,-0.2414],[316,-0.2414],[351,-0.2394],[379,-0.2394]]
       }
       CTRL_R_mouth_lipsTowardsTeethU {type=MovieSceneScriptingFloatChannel, default=0, const [0,17]}
       CTRL_faceGUIfollowHead {type=MovieSceneScriptingFloatChannel, default=1, const [1,17]}
       CTRL_eyesAimFollowHead {type=MovieSceneScriptingFloatChannel, default=0, const [0,17]}
       CTRL_C_nose.X {type=MovieSceneScriptingFloatChannel, default=0, const [0,17]}
       CTRL_C_nose.Y {type=MovieSceneScriptingFloatChannel, default=0, const [0,17]}
       CTRL_C_teeth_rotateU {type=MovieSceneScriptingFloatChannel, default=0, const [0,12]}
       CTRL_C_teeth_rotateD {type=MovieSceneScriptingFloatChannel, default=0, const [0,12]}
       CTRL_L_brow_inner {
        type=MovieSceneScriptingFloatChannel
        default=0
        thinned_from=15
        keys [[0,0],[19,0],[26,-0.6985],[85,-0.6985],[90,-1],[96,-0.1249],[101,-0.6985],[274,-0.6985]]
       }
       CTRL_R_brow_inner {
        type=MovieSceneScriptingFloatChannel
        default=0
        thinned_from=20
        keys [[0,0],[19,0],[26,-0.6985],[85,-0.6985],[90,-1],[96,-0.2507],[101,-0.6985],[114,-0.6985],[147,-0.524],[316,-0.524],[351,0],[379,0]]
       }
       CTRL_L_brow_outer {type=MovieSceneScriptingFloatChannel, default=0, thinned_from=15, keys [[0,0],[85,0],[90,-0.4367],[96,0.404],[101,0],[274,0]]}
       CTRL_R_brow_outer {type=MovieSceneScriptingFloatChannel, default=0, thinned_from=16, keys [[0,0],[85,0],[90,-0.6113],[96,0],[114,0],[147,0.1745],[274,0.1745]]}
       CTRL_C_mouth.X {type=MovieSceneScriptingFloatChannel, default=0, const [0,17]}
       CTRL_C_mouth.Y {type=MovieSceneScriptingFloatChannel, default=0, const [0,17]}
       CTRL_L_mouth_upperLipRaise {type=MovieSceneScriptingFloatChannel, default=0, const [0,17]}
       CTRL_L_mouth_cornerPull {type=MovieSceneScriptingFloatChannel, default=0, const [0,17]}
       CTRL_L_mouth_dimple {type=MovieSceneScriptingFloatChannel, default=0, const [0,17]}
       CTRL_L_mouth_cornerDepress {type=MovieSceneScriptingFloatChannel, default=0, const [0,17]}
       CTRL_L_mouth_stretch {type=MovieSceneScriptingFloatChannel, default=0, const [0,17]}
       CTRL_L_mouth_lowerLipDepress {type=MovieSceneScriptingFloatChannel, default=0, const [0,17]}
       CTRL_R_mouth_lowerLipDepress {type=MovieSceneScriptingFloatChannel, default=0, const [0,17]}
       CTRL_R_mouth_stretch {type=MovieSceneScriptingFloatChannel, default=0, const [0,17]}
       CTRL_L_mouth_towardsD {type=MovieSceneScriptingFloatChannel, default=0, const [0,17]}
       CTRL_R_mouth_funnelU {type=MovieSceneScriptingFloatChannel, default=0, const [0,17]}
       CTRL_L_mouth_towardsU {type=MovieSceneScriptingFloatChannel, default=0, const [0,17]}
       CTRL_R_mouth_lipBiteD {type=MovieSceneScriptingFloatChannel, default=0, const [0,17]}
       CTRL_R_mouth_funnelD {type=MovieSceneScriptingFloatChannel, default=0, const [0,17]}
       CTRL_L_mouth_pressU {type=MovieSceneScriptingFloatChannel, default=0, const [0,17]}
       CTRL_L_mouth_funnelD {type=MovieSceneScriptingFloatChannel, default=0, const [0,17]}
       CTRL_L_mouth_suckBlow {type=MovieSceneScriptingFloatChannel, default=0, const [0,16]}
       CTRL_L_mouth_pressD {type=MovieSceneScriptingFloatChannel, default=0, const [0,16]}
       CTRL_L_mouth_funnelU {type=MovieSceneScriptingFloatChannel, default=0, const [0,17]}
       CTRL_R_mouth_towardsU {type=MovieSceneScriptingFloatChannel, default=0, const [0,17]}
       CTRL_R_mouth_purseD {type=MovieSceneScriptingFloatChannel, default=0, const [0,17]}
       CTRL_L_mouth_lipBiteU {type=MovieSceneScriptingFloatChannel, default=0, const [0,17]}
       CTRL_R_mouth_pressD {type=MovieSceneScriptingFloatChannel, default=0, const [0,17]}
       CTRL_R_mouth_pressU {type=MovieSceneScriptingFloatChannel, default=0, const [0,17]}
       CTRL_L_mouth_purseD {type=MovieSceneScriptingFloatChannel, default=0, const [0,16]}
       CTRL_R_mouth_lipBiteU {type=MovieSceneScriptingFloatChannel, default=0, const [0,17]}
       CTRL_R_mouth_suckBlow {type=MovieSceneScriptingFloatChannel, default=0, const [0,16]}
       CTRL_L_mouth_purseU {type=MovieSceneScriptingFloatChannel, default=0, const [0,17]}
       CTRL_R_mouth_towardsD {type=MovieSceneScriptingFloatChannel, default=0, const [0,17]}
       CTRL_L_mouth_lipBiteD {type=MovieSceneScriptingFloatChannel, default=0, const [0,16]}
       CTRL_R_mouth_purseU {type=MovieSceneScriptingFloatChannel, default=0, const [0,16]}
       CTRL_R_mouth_cornerDepress {type=MovieSceneScriptingFloatChannel, default=0, const [0,17]}
       CTRL_R_mouth_upperLipRaise {type=MovieSceneScriptingFloatChannel, default=0, const [0,17]}
       CTRL_R_mouth_dimple {type=MovieSceneScriptingFloatChannel, default=0, const [0,17]}
       CTRL_R_mouth_cornerPull {type=MovieSceneScriptingFloatChannel, default=0, const [0,17]}
       Weight {type=MovieSceneScriptingFloatChannel, default=1}
      ]
     }]
   }
  ]
 }]
display_rate {denominator=1, numerator=30}
tick_resolution {denominator=1, numerator=24000}
tracks []
